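# Sound cue for the Cherry Smoke MaskMaker boss voice line set named GiveDmgLrg, apparently barks for dealing large damage.
# A Random node picks among six equally weighted voice wave variants.
# Playback uses the critical enemy VO sound class, a boosted volume and a shared NPC ethereal stereo attenuation.
# A boss combat chatter concurrency setting governs overlapping lines.
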SoundCue /MaskMakerBoss/Assets/Audio/VO/SC_CherrySmoke_MaskMaker_GiveDmgLrg
root=Random_1
nodes [
 {
  id=Random_1
  children [WavePlayer_24, WavePlayer_25, WavePlayer_26, WavePlayer_27, WavePlayer_28, WavePlayer_29]
  properties {weights [1,1,1,1,1,1]}
 }
 {id=WavePlayer_24, properties {soundWaveAssetPtr=/MaskMakerBoss/Assets/Audio/VO/Waves/VO_CherrySmoke_MaskMaker_GiveDmgLrg_01}}
 {id=WavePlayer_25, properties {soundWaveAssetPtr=/MaskMakerBoss/Assets/Audio/VO/Waves/VO_CherrySmoke_MaskMaker_GiveDmgLrg_02}}
 {id=WavePlayer_26, properties {soundWaveAssetPtr=/MaskMakerBoss/Assets/Audio/VO/Waves/VO_CherrySmoke_MaskMaker_GiveDmgLrg_03}}
 {id=WavePlayer_27, properties {soundWaveAssetPtr=/MaskMakerBoss/Assets/Audio/VO/Waves/VO_CherrySmoke_MaskMaker_GiveDmgLrg_04}}
 {id=WavePlayer_28, properties {soundWaveAssetPtr=/MaskMakerBoss/Assets/Audio/VO/Waves/VO_CherrySmoke_MaskMaker_GiveDmgLrg_05}}
 {id=WavePlayer_29, properties {soundWaveAssetPtr=/MaskMakerBoss/Assets/Audio/VO/Waves/VO_CherrySmoke_MaskMaker_GiveDmgLrg_06}}
]
properties {
 attenuationSettings=/Game/Sounds/Fort_Human_VO_BR/NPCs/Shared/NPC_VO_Alerted_Ethereal_Stereo_Att
 maxDistance=12500
 soundClassObject=&Fort_VO_Enemy_Critical
 volumeMultiplier=2.35
 concurrencySet [&NPC_Boss_Combat_Chatter_Conc]
}
