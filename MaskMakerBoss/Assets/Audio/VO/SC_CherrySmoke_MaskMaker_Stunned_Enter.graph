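# Sound cue for the CherrySmoke MaskMaker boss's voice lines, named for entering the stunned state.
# A Random node picks among fifteen equally weighted wave variations.
# It routes through the critical enemy VO sound class with a boosted volume, a shared NPC alerted stereo attenuation and boss combat chatter concurrency.
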SoundCue /MaskMakerBoss/Assets/Audio/VO/SC_CherrySmoke_MaskMaker_Stunned_Enter
root=Random_0
nodes [
 {
  id=Random_0
  children [
   WavePlayer_0
   WavePlayer_1
   WavePlayer_2
   WavePlayer_3
   WavePlayer_4
   WavePlayer_5
   WavePlayer_6
   WavePlayer_7
   WavePlayer_8
   WavePlayer_9
   WavePlayer_10
   WavePlayer_11
   WavePlayer_12
   WavePlayer_13
   WavePlayer_14
  ]
  properties {weights [1,1,1,1,1,1,1,1,1,1,1,1,1,1,1]}
 }
 {id=WavePlayer_0, properties {soundWaveAssetPtr=/MaskMakerBoss/Assets/Audio/VO/Waves/VO_CherrySmoke_MaskMaker_Stunned_Enter_01}}
 {id=WavePlayer_1, properties {soundWaveAssetPtr=/MaskMakerBoss/Assets/Audio/VO/Waves/VO_CherrySmoke_MaskMaker_Stunned_Enter_02}}
 {id=WavePlayer_10, properties {soundWaveAssetPtr=/MaskMakerBoss/Assets/Audio/VO/Waves/VO_CherrySmoke_MaskMaker_Stunned_Enter_11}}
 {id=WavePlayer_11, properties {soundWaveAssetPtr=/MaskMakerBoss/Assets/Audio/VO/Waves/VO_CherrySmoke_MaskMaker_Stunned_Enter_12}}
 {id=WavePlayer_12, properties {soundWaveAssetPtr=/MaskMakerBoss/Assets/Audio/VO/Waves/VO_CherrySmoke_MaskMaker_Stunned_Enter_13}}
 {id=WavePlayer_13, properties {soundWaveAssetPtr=/MaskMakerBoss/Assets/Audio/VO/Waves/VO_CherrySmoke_MaskMaker_Stunned_Enter_14}}
 {id=WavePlayer_14, properties {soundWaveAssetPtr=/MaskMakerBoss/Assets/Audio/VO/Waves/VO_CherrySmoke_MaskMaker_Stunned_Enter_15}}
 {id=WavePlayer_2, properties {soundWaveAssetPtr=/MaskMakerBoss/Assets/Audio/VO/Waves/VO_CherrySmoke_MaskMaker_Stunned_Enter_03}}
 {id=WavePlayer_3, properties {soundWaveAssetPtr=/MaskMakerBoss/Assets/Audio/VO/Waves/VO_CherrySmoke_MaskMaker_Stunned_Enter_04}}
 {id=WavePlayer_4, properties {soundWaveAssetPtr=/MaskMakerBoss/Assets/Audio/VO/Waves/VO_CherrySmoke_MaskMaker_Stunned_Enter_05}}
 {id=WavePlayer_5, properties {soundWaveAssetPtr=/MaskMakerBoss/Assets/Audio/VO/Waves/VO_CherrySmoke_MaskMaker_Stunned_Enter_06}}
 {id=WavePlayer_6, properties {soundWaveAssetPtr=/MaskMakerBoss/Assets/Audio/VO/Waves/VO_CherrySmoke_MaskMaker_Stunned_Enter_07}}
 {id=WavePlayer_7, properties {soundWaveAssetPtr=/MaskMakerBoss/Assets/Audio/VO/Waves/VO_CherrySmoke_MaskMaker_Stunned_Enter_08}}
 {id=WavePlayer_8, properties {soundWaveAssetPtr=/MaskMakerBoss/Assets/Audio/VO/Waves/VO_CherrySmoke_MaskMaker_Stunned_Enter_09}}
 {id=WavePlayer_9, properties {soundWaveAssetPtr=/MaskMakerBoss/Assets/Audio/VO/Waves/VO_CherrySmoke_MaskMaker_Stunned_Enter_10}}
]
properties {
 attenuationSettings=/Game/Sounds/Fort_Human_VO_BR/NPCs/Shared/NPC_VO_Alerted_Ethereal_Stereo_Att
 maxDistance=12500
 soundClassObject=&Fort_VO_Enemy_Critical
 volumeMultiplier=2.35
 concurrencySet [&NPC_Boss_Combat_Chatter_Conc]
}
