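# Sound cue of CherrySmoke MaskMaker boss voice lines, named for the exit from the stunned state.
# A root Random node picks one of 27 equally weighted wave players, each holding a numbered Stunned_Exit take.
# Playback is volume-boosted and uses a shared ethereal stereo NPC attenuation, the critical enemy VO sound class and the boss combat chatter concurrency set.
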SoundCue /MaskMakerBoss/Assets/Audio/VO/SC_CherrySmoke_MaskMaker_Stunned_Exit
root=Random_0
nodes [
 {
  id=Random_0
  children [
   WavePlayer_0
   WavePlayer_1
   WavePlayer_2
   WavePlayer_3
   WavePlayer_4
   WavePlayer_5
   WavePlayer_6
   WavePlayer_7
   WavePlayer_8
   WavePlayer_9
   WavePlayer_10
   WavePlayer_11
   WavePlayer_12
   WavePlayer_13
   WavePlayer_14
   WavePlayer_15
   WavePlayer_16
   WavePlayer_17
   WavePlayer_18
   WavePlayer_19
   WavePlayer_20
   WavePlayer_21
   WavePlayer_22
   WavePlayer_23
   WavePlayer_24
   WavePlayer_25
   WavePlayer_26
  ]
  properties {weights [1,1,1,1,1,1,1,1,1,1,1,1,1,1,1,1,1,1,1,1,1,1,1,1,1,1,1]}
 }
 {id=WavePlayer_0, properties {soundWaveAssetPtr=/MaskMakerBoss/Assets/Audio/VO/Waves/VO_CherrySmoke_MaskMaker_Stunned_Exit_01}}
 {id=WavePlayer_1, properties {soundWaveAssetPtr=/MaskMakerBoss/Assets/Audio/VO/Waves/VO_CherrySmoke_MaskMaker_Stunned_Exit_02}}
 {id=WavePlayer_10, properties {soundWaveAssetPtr=/MaskMakerBoss/Assets/Audio/VO/Waves/VO_CherrySmoke_MaskMaker_Stunned_Exit_11}}
 {id=WavePlayer_11, properties {soundWaveAssetPtr=/MaskMakerBoss/Assets/Audio/VO/Waves/VO_CherrySmoke_MaskMaker_Stunned_Exit_12}}
 {id=WavePlayer_12, properties {soundWaveAssetPtr=/MaskMakerBoss/Assets/Audio/VO/Waves/VO_CherrySmoke_MaskMaker_Stunned_Exit_13}}
 {id=WavePlayer_13, properties {soundWaveAssetPtr=/MaskMakerBoss/Assets/Audio/VO/Waves/VO_CherrySmoke_MaskMaker_Stunned_Exit_14}}
 {id=WavePlayer_14, properties {soundWaveAssetPtr=/MaskMakerBoss/Assets/Audio/VO/Waves/VO_CherrySmoke_MaskMaker_Stunned_Exit_15}}
 {id=WavePlayer_15, properties {soundWaveAssetPtr=/MaskMakerBoss/Assets/Audio/VO/Waves/VO_CherrySmoke_MaskMaker_Stunned_Exit_16}}
 {id=WavePlayer_16, properties {soundWaveAssetPtr=/MaskMakerBoss/Assets/Audio/VO/Waves/VO_CherrySmoke_MaskMaker_Stunned_Exit_17}}
 {id=WavePlayer_17, properties {soundWaveAssetPtr=/MaskMakerBoss/Assets/Audio/VO/Waves/VO_CherrySmoke_MaskMaker_Stunned_Exit_18}}
 {id=WavePlayer_18, properties {soundWaveAssetPtr=/MaskMakerBoss/Assets/Audio/VO/Waves/VO_CherrySmoke_MaskMaker_Stunned_Exit_19}}
 {id=WavePlayer_19, properties {soundWaveAssetPtr=/MaskMakerBoss/Assets/Audio/VO/Waves/VO_CherrySmoke_MaskMaker_Stunned_Exit_20}}
 {id=WavePlayer_2, properties {soundWaveAssetPtr=/MaskMakerBoss/Assets/Audio/VO/Waves/VO_CherrySmoke_MaskMaker_Stunned_Exit_03}}
 {id=WavePlayer_20, properties {soundWaveAssetPtr=/MaskMakerBoss/Assets/Audio/VO/Waves/VO_CherrySmoke_MaskMaker_Stunned_Exit_21}}
 {id=WavePlayer_21, properties {soundWaveAssetPtr=/MaskMakerBoss/Assets/Audio/VO/Waves/VO_CherrySmoke_MaskMaker_Stunned_Exit_22}}
 {id=WavePlayer_22, properties {soundWaveAssetPtr=/MaskMakerBoss/Assets/Audio/VO/Waves/VO_CherrySmoke_MaskMaker_Stunned_Exit_23}}
 {id=WavePlayer_23, properties {soundWaveAssetPtr=/MaskMakerBoss/Assets/Audio/VO/Waves/VO_CherrySmoke_MaskMaker_Stunned_Exit_24}}
 {id=WavePlayer_24, properties {soundWaveAssetPtr=/MaskMakerBoss/Assets/Audio/VO/Waves/VO_CherrySmoke_MaskMaker_Stunned_Exit_25}}
 {id=WavePlayer_25, properties {soundWaveAssetPtr=/MaskMakerBoss/Assets/Audio/VO/Waves/VO_CherrySmoke_MaskMaker_Stunned_Exit_26}}
 {id=WavePlayer_26, properties {soundWaveAssetPtr=/MaskMakerBoss/Assets/Audio/VO/Waves/VO_CherrySmoke_MaskMaker_Stunned_Exit_27}}
 {id=WavePlayer_3, properties {soundWaveAssetPtr=/MaskMakerBoss/Assets/Audio/VO/Waves/VO_CherrySmoke_MaskMaker_Stunned_Exit_04}}
 {id=WavePlayer_4, properties {soundWaveAssetPtr=/MaskMakerBoss/Assets/Audio/VO/Waves/VO_CherrySmoke_MaskMaker_Stunned_Exit_05}}
 {id=WavePlayer_5, properties {soundWaveAssetPtr=/MaskMakerBoss/Assets/Audio/VO/Waves/VO_CherrySmoke_MaskMaker_Stunned_Exit_06}}
 {id=WavePlayer_6, properties {soundWaveAssetPtr=/MaskMakerBoss/Assets/Audio/VO/Waves/VO_CherrySmoke_MaskMaker_Stunned_Exit_07}}
 {id=WavePlayer_7, properties {soundWaveAssetPtr=/MaskMakerBoss/Assets/Audio/VO/Waves/VO_CherrySmoke_MaskMaker_Stunned_Exit_08}}
 {id=WavePlayer_8, properties {soundWaveAssetPtr=/MaskMakerBoss/Assets/Audio/VO/Waves/VO_CherrySmoke_MaskMaker_Stunned_Exit_09}}
 {id=WavePlayer_9, properties {soundWaveAssetPtr=/MaskMakerBoss/Assets/Audio/VO/Waves/VO_CherrySmoke_MaskMaker_Stunned_Exit_10}}
]
properties {
 attenuationSettings=/Game/Sounds/Fort_Human_VO_BR/NPCs/Shared/NPC_VO_Alerted_Ethereal_Stereo_Att
 maxDistance=12500
 soundClassObject=&Fort_VO_Enemy_Critical
 volumeMultiplier=2.35
 concurrencySet [&NPC_Boss_Combat_Chatter_Conc]
}
