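# Sound cue for the Mask Maker boss's small-damage voice reaction, from the CherrySmoke VO set.
# A random node picks one of eight equally weighted wave variations.
# It plays in the critical enemy VO sound class with shared NPC alerted attenuation and boosted volume.
# A concurrency override allowing one instance and blocking new plays is enabled, alongside an assigned boss combat chatter concurrency set.
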SoundCue /MaskMakerBoss/Assets/Audio/VO/SC_CherrySmoke_MaskMaker_TakeDmgSml
root=Random_0
nodes [
 {
  id=Random_0
  children [WavePlayer_0,WavePlayer_1,WavePlayer_2,WavePlayer_3,WavePlayer_4,WavePlayer_5,WavePlayer_6,WavePlayer_7]
  properties {weights [1,1,1,1,1,1,1,1]}
 }
 {id=WavePlayer_0, properties {soundWaveAssetPtr=/MaskMakerBoss/Assets/Audio/VO/Waves/VO_CherrySmoke_MaskMaker_TakeDmgSml_01}}
 {id=WavePlayer_1, properties {soundWaveAssetPtr=/MaskMakerBoss/Assets/Audio/VO/Waves/VO_CherrySmoke_MaskMaker_TakeDmgSml_02}}
 {id=WavePlayer_2, properties {soundWaveAssetPtr=/MaskMakerBoss/Assets/Audio/VO/Waves/VO_CherrySmoke_MaskMaker_TakeDmgSml_03}}
 {id=WavePlayer_3, properties {soundWaveAssetPtr=/MaskMakerBoss/Assets/Audio/VO/Waves/VO_CherrySmoke_MaskMaker_TakeDmgSml_04}}
 {id=WavePlayer_4, properties {soundWaveAssetPtr=/MaskMakerBoss/Assets/Audio/VO/Waves/VO_CherrySmoke_MaskMaker_TakeDmgSml_05}}
 {id=WavePlayer_5, properties {soundWaveAssetPtr=/MaskMakerBoss/Assets/Audio/VO/Waves/VO_CherrySmoke_MaskMaker_TakeDmgSml_06}}
 {id=WavePlayer_6, properties {soundWaveAssetPtr=/MaskMakerBoss/Assets/Audio/VO/Waves/VO_CherrySmoke_MaskMaker_TakeDmgSml_07}}
 {id=WavePlayer_7, properties {soundWaveAssetPtr=/MaskMakerBoss/Assets/Audio/VO/Waves/VO_CherrySmoke_MaskMaker_TakeDmgSml_08}}
]
properties {
 attenuationSettings=/Game/Sounds/Fort_Human_VO_BR/NPCs/Shared/NPC_VO_Alerted_Ethereal_Stereo_Att
 bOverrideConcurrency=true
 maxDistance=12500
 soundClassObject=&Fort_VO_Enemy_Critical
 volumeMultiplier=2.2
 concurrencyOverrides {maxCount=1, resolutionRule=PreventNew}
 concurrencySet [&NPC_Boss_Combat_Chatter_Conc]
}
